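# Translucent master material for tinted Creative glass, faking reflections with a cubemap sampled along the reflection vector.
# The brightened, desaturated cubemap is added to the tint Color for base color and multiplied by it for emissive.
# Fresnel adds to opacity and blends refraction from 1 toward an adjustable index.
# Attributes pass through MF_CreativeMoveTool, and a material proxy branch substitutes a flat normal for the normal map parameter.
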Material /Game/Creative/Environments/Meshes/Glass/Materials/Glass_version_02/M_CP_Glass_common_Master
graph {
 properties {
  bUseMaterialAttributes=true
  blendMode=BLEND_Translucent
  pixelDepthOffsetMode=PDOM_Legacy
  refractionMethod=RM_IndexOfRefraction
  usedShadingModels=Default Lit
 }
 nodes [
  n1 = VectorParameter() {parameterName=Color, defaultValue=<rgba 0.2962 0.4347 0.9896 1>}
  n2 = TextureSampleParameterCube!(UVs=ReflectionVectorWS(), Apply View MipBias) {
   parameterName=Reflection Cubemap
   samplerType=SAMPLERTYPE_LinearColor
   texture=/Game/Creative/Environments/Meshes/Glass/Textures/T_Glass_Cubemap_01
  }
  n3 = Desaturation(None=Multiply(A=@n2.RGB, B=ScalarParameter() {parameterName=Cube_brightness}), Fraction=ScalarParameter() {parameterName=Cube_desaturation})
  n4 = Multiply(A=@n3, B=@n1.RGB)
  n5 = ScalarParameter() {defaultValue=1, parameterName=Emissive Multiplier}
  n6 = ScalarParameter() {parameterName=Opacity_intensity}
  n7 = Fresnel(ExponentIn=ScalarParameter() {parameterName=Fresnel_Exponent}, BaseReflectFractionIn=ScalarParameter() {parameterName=Fresnel_Base})
  n8 = TextureCoordinate()
  n9 = Constant() {r=1}
  n10 = ScalarParameter() {parameterName=ReFraction_Adjust}
  n11 = MaterialFunctionCall(MA In=MakeMaterialAttributes!(
    BaseColor=Add(A=@n3, B=@n1.RGB)
    Metallic
    Specular
    Roughness
    Anisotropy
    EmissiveColor=Multiply(A=@n4, B=@n5)
    Opacity=Add(A=@n6, B=@n7)
    OpacityMask
    Normal=TextureSampleParameter2D(UVs=@n8) {parameterName=Normal, samplerType=SAMPLERTYPE_Normal, texture=&FlatNormalBlue}.RGB
    Tangent
    WorldPositionOffset
    SubsurfaceColor
    ClearCoat
    ClearCoatRoughness
    AmbientOcclusion
    Refraction=LinearInterpolate(A=@n9, B=@n10, Alpha=@n7)
    CustomizedUVs_0
    CustomizedUVs_1
    CustomizedUVs_2
    CustomizedUVs_3
    CustomizedUVs_4
    CustomizedUVs_5
    CustomizedUVs_6
    CustomizedUVs_7
    PixelDepthOffset
    ShadingModel
    Displacement
   )) {materialFunction=&MF_CreativeMoveTool}
  MaterialProxyReplace(
   Realtime=@n11.MA Out
   MaterialProxy=SetMaterialAttributes!(
    MaterialAttributes=@n11.MA Out
    Normal=TextureSample() {samplerType=SAMPLERTYPE_Normal, texture=&FlatNormalBlue}.RGB
   ) {attributeSetTypes [#1]}
  )
 ]
}
